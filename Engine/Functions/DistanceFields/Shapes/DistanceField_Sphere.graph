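# DistanceField_Sphere is a library material function that returns a signed distance field for a sphere.
# It subtracts Center from Coordinates, measures that offset's distance to zero, then subtracts Radius, so values are negative inside the sphere.
# Coordinates default to UVs extended with a zero Z.
# Its description text mentions drawing a line between two points, which does not match the sphere math.
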MaterialFunction /Engine/Functions/DistanceFields/Shapes/DistanceField_Sphere
graph {
 properties {bExposeToLibrary=true, description="Draws a line between 2 points using supplied 3D coordinates. ", libraryCategoriesText [Distance Fields]}
 nodes [
  n1 = TextureCoordinate() {desc=UVs}
  n2 = Constant()
  n3 = Subtract(
   A=FunctionInput(Preview=AppendVector(A=@n1, B=@n2)) {bUsePreviewValueAsDefault=true, inputName=Coordinates}
   B=FunctionInput() {bUsePreviewValueAsDefault=true, inputName=Center, sortPriority=1}
  )
  n4 = Constant()
  n5 = FunctionInput() {bUsePreviewValueAsDefault=true, inputName=Radius, inputType=FunctionInput_Scalar, sortPriority=2, previewValue=<0.005 0 0 1>}
  FunctionOutput(None=Subtract(A=Distance(A=@n3, B=@n4), B=@n5))
 ]
}
