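# DistanceField_Capsule is a library material function that returns a signed distance field for a capsule, or thick line segment, between two 3D points.
# It projects Coordinates onto the P0-to-P1 segment with a dot-product ratio, clamped to keep the closest point between the endpoints.
# The distance to that point minus Width is the output.
# Coordinates preview as UVs with an appended constant.
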MaterialFunction /Engine/Functions/DistanceFields/Shapes/DistanceField_Capsule
graph {
 properties {bExposeToLibrary=true, description="Draws a line between 2 points using supplied 3D coordinates. ", libraryCategoriesText [Distance Fields]}
 nodes [
  n1 = TextureCoordinate() {desc=UVs}
  n2 = Constant()
  n3 = FunctionInput() {bUsePreviewValueAsDefault=true, inputName=P0, sortPriority=1}
  n4 = Subtract(A=FunctionInput(Preview=AppendVector(A=@n1, B=@n2)) {bUsePreviewValueAsDefault=true, inputName=Coordinates}, B=@n3) {desc=PA}
  n5 = Subtract(A=FunctionInput() {bUsePreviewValueAsDefault=true, inputName=P1, sortPriority=2, previewValue=<1 1 0 1>}, B=@n3) {desc=BA}
  n6 = Divide(A=DotProduct(A=@n4, B=@n5), B=DotProduct(A=@n5, B=@n5)) {constB=1}
  n7 = Subtract(A=@n4, B=Multiply(A=Clamp(None=@n6), B=@n5))
  n8 = Constant()
  n9 = FunctionInput() {bUsePreviewValueAsDefault=true, inputName=Width, inputType=FunctionInput_Scalar, sortPriority=3, previewValue=<0.005 0 0 1>}
  FunctionOutput(None=Subtract(A=Distance(A=@n7, B=@n8), B=@n9))
 ]
}
